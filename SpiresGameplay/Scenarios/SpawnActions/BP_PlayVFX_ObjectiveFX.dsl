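; Spawn action for Spires scenarios that gives a spawned actor objective VFX.
; When an actor spawns, it checks validity and adds a BP_ReplicatedPawnVFXComponent_ObjectiveFX component to that actor.
; The effect itself is left to that component class, and the source point provider input is unused.
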(fn OnActorSpawned
  (params (SpawnedActor :type Actor%20Object%20Reference) (SourcePointProvider :type Fort%20Athena%20Living%20World%20Point%20Provider%20Interface%20Interface%20%28by%20ref%29))
  (Utilities|IsValid SpawnedActor
    (:Is%20Valid
      (Game|AddComponentbyClass SpawnedActor "/SpiresGameplay/Scenarios/SpawnActions/BP_ReplicatedPawnVFXComponent_ObjectiveFX.BP_ReplicatedPawnVFXComponent_ObjectiveFX_C")
      (return))))
