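; Pawn component that attaches an objective Niagara effect to its owner, using the location of the actor that owns the pawn's controller.
; On BeginPlay, only the authority resolves that controller owner, as clients cannot, and passes its location to the Spawn FX event.
; Spawn FX attaches the VFX system to the root component and feeds that location, unmodified, into the SystemDirection Niagara parameter.
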(event EventBeginPlay
  (bind _returnvalue (Components|GetOwner))
  (bind _returnvalue_2 (Fortnite|GetControllerfromActor _returnvalue))
  (bind _returnvalue_3 (Actor|GetOwner _returnvalue_2))
  ; We can only get the controller/owner from the server, whatever attributes we need from the owner we would need to replicate to clients
  (Utilities|SwitchHasAuthority
    (:Authority
      (Utilities|IsValid _returnvalue_3
        (:Is%20Valid
          (CallFunction|SpawnFX (Transformation|GetActorLocation _returnvalue_3)))))))

(event Custom|Spawn%20FX
  (params (OwnerLocation :type Vector))
  ; We can only get the controller/owner from the server, whatever attributes we need from the owner we would need to replicate to clients
  (Variables|Default|SetSystemDirection OwnerLocation)
  (bind _returnvalue (Niagara|SpawnSystemAttached (Variables|Default|GetVFX) (Variables|/Script/Engine.Actor|GetRootComponent (Components|GetOwner))))
  (bind _output_get (Variables|Default|SetNiagaraComp _returnvalue))
  (Niagara|SetNiagaraVariable(Vector3) _output_get "SystemDirection" (Variables|Default|GetSystemDirection)))
