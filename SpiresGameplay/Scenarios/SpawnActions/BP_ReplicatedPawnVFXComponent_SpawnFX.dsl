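; Actor component Blueprint that spawns an attached Niagara effect on its owning replicated pawn when play begins.
; On BeginPlay, only the authority proceeds, and only if the owner's controller has a valid owner, since controller data is available only on the server.
; The Spawn FX custom event attaches the system returned by GetVFX to the owning actor's root component.
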(event Custom|Spawn%20FX
  (Niagara|SpawnSystemAttached (Variables|Default|GetVFX) (Variables|/Script/Engine.Actor|GetRootComponent (Components|GetOwner))))

(event EventBeginPlay
  ; We can only get the controller/owner from the server, whatever attributes we need from the owner we would need to replicate to clients
  (Utilities|SwitchHasAuthority
    (:Authority
      (Utilities|IsValid (Actor|GetOwner (Fortnite|GetControllerfromActor (Components|GetOwner)))
        (:Is%20Valid
          (CallFunction|SpawnFX))))))
